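# Additive Niagara sprite material for wobbling, iridescent soap bubbles in the main player's Lifted emote FX.
# Panning noise, scaled by dynamic parameters, distorts bubble UVs and a radial gradient.
# That gradient samples an iridescence curve atlas, hue-shifted by offset time sines, for emissive color alongside the bubble texture.
# Bubble channels, normal-map Fresnel and particle alpha build a mask; mobile swaps particle randomness for constants.
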Material /Game/Animation/Game/MainPlayer/Emotes/Lifted/FX/M_Soap_Bubbles
graph {
 properties {
  bUsedWithNiagaraSprites=true
  blendMode=BLEND_Additive
  pixelDepthOffsetMode=PDOM_Legacy
  refractionMethod=RM_IndexOfRefraction
  usedShadingModels=Default Lit
 }
 nodes [
  n1 = Multiply(A=Time() {period=1}) {constB=0.1}
  n2 = ParticleRandom()
  n3 = Constant() {r=0.015}
  n4 = Add(A=ShadingPathSwitch(Default=Add(A=@n1, B=@n2), Mobile=Add(A=@n1, B=@n3)) {desc=Migrated from FeatureLevelSwitch})
  n5 = Add(A=@n4) {constB=0}
  n6 = ScalarParameter() {defaultValue=0.35, parameterName=SeperationGap}
  n7 = Add(A=@n4, B=@n6) {constB=0.1}
  n8 = Multiply(A=@n6) {constB=2}
  n9 = AppendVector(A=AppendVector(A=Sine(None=@n5), B=Sine(None=@n7)), B=Sine(None=Add(A=@n4, B=@n8) {constB=0.15}))
  n10 = MaterialFunctionCall(Float2=TextureCoordinate()) {materialFunction=&BreakOutFloat2Components}
  n11 = ScalarParameter() {defaultValue=1, parameterName=Base_U_Tile}
  n12 = ScalarParameter() {defaultValue=1, parameterName=Base_V_Tile}
  n13 = TextureCoordinate()
  n14 = MaterialFunctionCall(Float2=TextureCoordinate()) {materialFunction=&BreakOutFloat2Components}
  n15 = DynamicParameter() {paramNames [DistortionX,DistortionY,Param3,Param4]}
  n16 = Add(
   A=AppendVector(A=Multiply(A=@n14.R, B=@n15.DistortionX), B=Multiply(A=@n14.G, B=@n15.DistortionY))
   B=ShadingPathSwitch(Default=ParticleRandom(), Mobile=Constant()) {desc=Migrated from FeatureLevelSwitch}
  )
  n17 = AppendVector(
   A=ScalarParameter() {defaultValue=0.2, parameterName=Distort_U_PanSpeed}
   B=ScalarParameter() {defaultValue=0.15, parameterName=Distort_V_PanSpeed}
  )
  n18 = MaterialFunctionCall() {materialFunction=&LinearGradient}
  n19 = LinearInterpolate(
   B=TextureSample(UVs=Panner(Coordinate=@n16, Speed=@n17) {speedX=0.02, speedY=0.15}) {samplerType=SAMPLERTYPE_Grayscale, texture=&LowResBlurredNoise}.G
   Alpha=LinearInterpolate(A=SquareRoot(None=@n18.UGradient), Alpha=ScalarParameter() {defaultValue=1, parameterName=Lerp_GradientBlendDistortNoise})
  )
  n20 = Add(A=AppendVector(A=Multiply(A=@n10.R, B=@n11), B=Multiply(A=@n10.G, B=@n12)), B=Multiply(A=Add(A=@n13) {constB=-0.5}, B=@n19))
  n21 = VectorParameter() {parameterName=SphereMaskPosition, defaultValue=<rgba 0.5 0.5 0 0>}
  n22 = MaterialFunctionCall!(
   UVs=@n20
   CenterPosition=LinearInterpolate(A=ComponentMask(None=@n21.RGB) {g=true, r=true}, B=Constant2Vector() {g=0.5, r=0.5}, Alpha=@n19)
   Radius=ScalarParameter() {defaultValue=0.5, parameterName=SphereMaskRadius}
  ) {materialFunction=/Engine/Functions/Engine_MaterialFunctions01/Gradient/SphereGradient-2D}
  n23 = CurveAtlasRowParameter(CurveTime=Power(Base=@n22.Result 0-1) {constExponent=0.5}) {
   atlas=&/A_Curve_Bubbles
   curve=&/C_Curve_Irridescent
  }
  n24 = MaterialFunctionCall(Hue Shift Percentage=@n9, Texture=@n23) {materialFunction=&HueShift}
  Add(A=Multiply(A=@n9, B=@n22.Result 0-1), B=@n24.Result)
  Fresnel()
  LinearInterpolate(B=ScalarParameter() {defaultValue=0.95, parameterName=Refraction}, Alpha=@n22.Result 0-1) {constA=1}
  n25 = TextureSample(UVs=@n20) {samplerSource=SSM_Clamp_WorldGroupSettings, texture=&/T_Bubble}
  MaterialFunctionCall(Emissive=Add(A=@n25.R, B=Power(Base=@n24.Result) {constExponent=0.75})) {materialFunction=&MF_FX_EmissiveAdjustments}
  n26 = OneMinus(None=MaterialFunctionCall(In=@n22.Result 0-1, Contrast=Constant() {r=1.25}) {materialFunction=&CheapContrast}.Result)
  n27 = TextureSample(UVs=@n20) {
   samplerSource=SSM_Clamp_WorldGroupSettings
   samplerType=SAMPLERTYPE_Normal
   texture=&/T_Bubble_N
  }
  n28 = Multiply(A=@n26, B=Fresnel(ExponentIn=Constant() {r=5}, Normal=@n27.RGB))
  n29 = Multiply(A=@n26, B=Power(Base=Multiply(A=@n28, B=@n22.Result 0-1)) {constExponent=0.25})
  n30 = Add(A=@n25.R, B=Add(A=Power(Base=@n25.B) {constExponent=3}, B=Multiply(A=@n25.B, B=@n29)))
  n31 = ParticleColor()
  Multiply(A=Power(Base=Multiply(A=@n30, B=@n31.A)) {constExponent=0.45})
  ScalarParameter() {defaultValue=0.2, parameterName=Distort_U_Tile}
  ScalarParameter() {defaultValue=0.2, parameterName=Distort_V_Tile}
  SphereMask()
 ]
}
