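# Niagara sprite effect for the player's "Lifted" emote, with a rising puff emitter and a water-drop emitter.
# SmallerSmoke001 spawns continuously, drifts upward with curl noise, swells, fades sharply at end of life and emits death events.
# FN_Water_Drops_ has short-lived drops pushed from a point, falling and shrinking, with no spawn module or event handler shown.
# Water, front-end and color user parameters are declared but unreferenced in the shown modules.
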NiagaraSystem /Game/Animation/Game/MainPlayer/Emotes/Lifted/FX/NS_Lifted
dependencies {
 usedDataInterfaces [&NiagaraDataInterfaceCurlNoise, &NiagaraDataInterfaceCurve]
 usedDynamicInputs []
 usedModules [&SystemState]
}
emitters {
 FN_Water_Drops_ {
  emitterUpdateScript [&EmitterState(Life Cycle Mode=::NewEnumerator1, Loop Delay=0.9083, Loop Duration=0, Recalculate Duration Each Loop=true, UseLoopDelay=false) {module=EmitterState001}]
  particleSpawnScript [
   /Niagara/Modules/Spawn/Initialization/InitializeParticle(
    Color=&RandomRangeLinearColor(Minimum=<rgba 1 1 1 1>, Maximum=<rgba 1 1 1.2 1.25>)
    Lifetime=&UniformRangedFloat(Minimum=0.25, Maximum=0.35)
    Mass=&UniformRangedFloat(Minimum=0.6, Maximum=1.2)
    Mesh Scale=<100 100 100>
    Module.Write Color=true
    Module.Write Lifetime=true
    Module.Write Mass=false
    Module.Write Material Random=true
    Module.Write Position=true
    Module.Write Scale=false
    Module.Write SpriteRotation=false
    Module.Write SpriteSize=true
    Position=&SimulationPosition()
    Sprite Size=&RandomRangeVector2D(Minimum=<1.5 2.75>, Maximum=<2 3.25>)
   )
   &AddVelocityFromPoint(Module.Use Falloff Distance=false)
  ]
  particleUpdateScript [
   &ParticleState()
   &CurlNoiseForce(Noise Strength=250, Pan Noise=false)
   &ScaleColor(
    Scale Alpha=&FloatFromCurve(FloatCurve=DataInterface() {propertyValues {Curve {keys [
         {arriveTangent=1.8, interpMode=RCIM_Cubic, leaveTangent=1.8, time=0, value=0}
         {arriveTangent=0.002, interpMode=RCIM_Cubic, leaveTangent=0.002, time=0.5, value=0.9}
         {arriveTangent=-1.792, interpMode=RCIM_Cubic, leaveTangent=-1.792, time=1.0011, value=0.002}
        ]}}})
    ScaleA=true
    ScaleRGB=true
   )
   &Drag(Drag=&RandomRangeFloat(Minimum=0.5, Maximum=1), Use Linear Drag=true, Use Rotational Drag=false)
   &GravityForce(Gravity=&RandomRangeVector(Minimum=<0 0 -128>, Maximum=<0 0 -256>))
   &SolveForcesAndVelocity(Clamp Velocity=false, Limit Acceleration=false)
   &ScaleSpriteSize(Uniform Curve Sprite Scale=DataInterface() {propertyValues {Curve {keys [
        {tangentMode=RCTM_User, time=0, value=1}
        {arriveTangent=-2.3165, interpMode=RCIM_Cubic, leaveTangent=-2.3165, tangentMode=RCTM_User, time=0.5, value=0.65}
        {arriveTangent=0.491, interpMode=RCIM_Cubic, leaveTangent=0.491, tangentMode=RCTM_User, time=1, value=0}
       ]}}})
  ]
  rendererClasses [&NiagaraSpriteRendererProperties]
  renderers [{rendererClass=&NiagaraSpriteRendererProperties}]
 }
 SmallerSmoke001 {
  emitterUpdateScript [
   &EmitterState(Life Cycle Mode=::NewEnumerator1, Loop Delay=0.9083, Loop Duration=0, Recalculate Duration Each Loop=true, UseLoopDelay=false)
   &SpawnRate(SpawnRate=&RandomRangeFloat(Minimum=24, Maximum=48), Use Spawn Probability=false)
  ]
  particleSpawnScript [
   /Niagara/Modules/Spawn/Initialization/V2/InitializeParticle(
    Color Mode=::NewEnumerator0
    Lifetime=&RandomRangeFloat(Minimum=0.4, Maximum=1)
    Position Offset=<0 0 10>
    Position Offset Coordinate Space=::Local
    Sprite Rotation Angle=&RandomRangeFloat(Minimum=-0.2, Maximum=-0.1)
    Sprite Rotation Mode=::NewEnumerator1
    Sprite Size Mode=::NewEnumerator1
    Uniform Sprite Size Max=5
    Uniform Sprite Size Min=1
    UsePositionOffset=true
    Write Material Random=true
   )
   &AddVelocity(Velocity=&RandomRangeVector(Minimum=<0 -100 50>, Maximum=<0 -200 100>))
   &BoxLocation(Box Size=<5 15 5>, MaskSpawnGroup=false, Module.Override Seed=false, Rotate Box=false, Surface Only=false)
  ]
  particleUpdateScript [
   &CurlNoiseForce(Noise Strength=250, Pan Noise=false)
   &GravityForce(Gravity=<0 0 50>)
   &DynamicMaterialParameters(
    Index 0 Param 1=&FloatFromCurve(FloatCurve=DataInterface() {propertyValues {Curve {keys [
         {arriveTangent=0.348, interpMode=RCIM_Cubic, leaveTangent=0.348, tangentMode=RCTM_User, time=0, value=0.2}
         {arriveTangent=0.3293, interpMode=RCIM_Cubic, leaveTangent=0.3293, tangentMode=RCTM_User, time=0.9, value=0.25}
         {arriveTangent=3.2005, interpMode=RCIM_Cubic, leaveTangent=3.2005, tangentMode=RCTM_User, time=1, value=0.75}
        ]}}})
    Module.Param0WriteEnabled=true
    Module.Param1WriteEnabled=true
    Module.Param2WriteEnabled=true
    Module.Param3WriteEnabled=true
   )
   &Drag(Drag=&RandomRangeFloat(Minimum=1.5, Maximum=2.5), Use Linear Drag=true, Use Rotational Drag=false)
   &InheritSourceMovement(Applied Position Scale=<0.5 0.5 0.75>, ApplyForce=false, ApplyPosition=true, ApplyPositionFalloff=false)
   &AccelerationForce(Acceleration=&RandomRangeVector(Minimum=<-100 -100 0>, Maximum=<100 100 0>))
   &SolveForcesAndVelocity(Clamp Velocity=false, Limit Acceleration=false)
   &Color(
    Color=&ColorFromCurve(ColorCurve=DataInterface() {propertyValues {
       AlphaCurve {keys [<at 0.965 1>, <at 0.9737 1>, <at 0.9877 0.2>, {arriveTangent=-16.2859, interpMode=RCIM_Cubic, leaveTangent=-16.2859, time=1, value=0}]}
       BlueCurve {keys [
         {arriveTangent=-1, interpMode=RCIM_Cubic, leaveTangent=-1, time=0, value=1}
         <at 0.965 1>
         {arriveTangent=-1, interpMode=RCIM_Cubic, leaveTangent=-1, time=1, value=0}
        ]}
       GreenCurve {keys [
         {arriveTangent=-1, interpMode=RCIM_Cubic, leaveTangent=-1, time=0, value=1}
         <at 0.965 1>
         {arriveTangent=-1, interpMode=RCIM_Cubic, leaveTangent=-1, time=1, value=0}
        ]}
       RedCurve {keys [
         {arriveTangent=-1, interpMode=RCIM_Cubic, leaveTangent=-1, time=0, value=1}
         <at 0.965 1>
         {arriveTangent=-1, interpMode=RCIM_Cubic, leaveTangent=-1, time=1, value=0}
        ]}
      }})
    Scale Alpha=&FloatFromCurve(FloatCurve=DataInterface() {propertyValues {Curve {keys [<at 0.8 1>,<at 1 0>]}}}, Scale Curve=0.6)
   )
   &ParticleState()
   &ScaleSpriteSize(Scale Factor=&Vector2DFromFloat(Value=&FloatFromCurve(
      FloatCurve=DataInterface() {propertyValues {Curve {keys [
          {arriveTangent=4.1785, interpMode=RCIM_Cubic, leaveTangent=4.1785, tangentMode=RCTM_User, time=0, value=0}
          {arriveTangent=0.9995, interpMode=RCIM_Cubic, leaveTangent=0.9995, tangentMode=RCTM_User, time=0.1, value=0.35}
          <at 0.9 0.75>
          {arriveTangent=2.5, interpMode=RCIM_Cubic, leaveTangent=2.5, time=0.95, value=0.57}
          {arriveTangent=4.4385, interpMode=RCIM_Cubic, leaveTangent=4.4385, tangentMode=RCTM_User, time=1, value=1}
         ]}}}
      Scale Curve=4
     )))
   &SpriteRotationRate(Rotation Rate=&RandomRangeFloat(Minimum=-40, Maximum=40))
   &GenerateDeathEvent(Event Probability=1, Use Event Probability=true)
  ]
  rendererClasses [&NiagaraSpriteRendererProperties]
  renderers [{rendererClass=&NiagaraSpriteRendererProperties}]
 }
}
summary {
 systemName=NS_Lifted
 emitters [
  {emitterName=SmallerSmoke001, rendererClasses [&NiagaraSpriteRendererProperties]}
  {emitterName=FN_Water_Drops_, rendererClasses [&NiagaraSpriteRendererProperties]}
 ]
 userVariables [
  User.bIsFrontEnd {type=NiagaraBool}
  User.bIsFrontEndPreview {type=NiagaraBool}
  User.bIsInWater {type=NiagaraBool}
  User.FireworkColors {
   type=NiagaraDataInterfaceColorCurve
   description=""
   defaultValue {
    struct=NiagaraExt_VariableValue_DataInterface
    value {
     dataInterface=&self:NiagaraDataInterfaceColorCurve_0
     dataInterfaceClass=&NiagaraDataInterfaceColorCurve
    }
   }
  }
  User.PlayerSpeed {type=NiagaraFloat, defaultValue=0, description=""}
  User.Skeletal Mesh01 {
   type=NiagaraDataInterfaceSkeletalMesh
   defaultValue {
    struct=NiagaraExt_VariableValue_DataInterface
    value {dataInterface=&self:NiagaraDataInterfaceSkeletalMesh_2}
   }
  }
  User.WaterBodyIndex {type=NiagaraFloat, defaultValue=0, description=""}
  User.WaterDepth {type=NiagaraFloat, defaultValue=0, description=""}
  User.WaterPlaneLocation {type=Vector3f, description="", defaultValue {struct=Vector3f, value=<0 0 0>}}
  User.WaterPlaneNormal {type=Vector3f, description="", defaultValue {struct=Vector3f, value=<0 0 0>}}
  User.WaveAttenuationFactor {type=NiagaraFloat, defaultValue=0, description=""}
  User.WaveTime {type=NiagaraFloat, defaultValue=0, description=""}
 ]
}
